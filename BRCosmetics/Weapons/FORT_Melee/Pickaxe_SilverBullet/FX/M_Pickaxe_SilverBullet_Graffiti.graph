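# Translucent Niagara sprite material that draws graffiti for the Silver Bullet pickaxe.
# Two flipbook samples of a packed mask texture, indexed by dynamic parameters, form background and text, colored through curve atlas rows and tinted by particle color.
# Dynamic parameters drive a smoothstep text reveal and soft erosion of opacity.
# The erosion texture rotates by a signed angle computed from the X Vector parameter using cross and dot products.
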Material /BRCosmetics/Weapons/FORT_Melee/Pickaxe_SilverBullet/FX/M_Pickaxe_SilverBullet_Graffiti
graph {
 properties {bAutomaticallySetUsageInEditor=false, bUsedWithNiagaraSprites=true, blendMode=BLEND_Translucent, usedShadingModels=Default Lit}
 nodes [
  Constant3Vector() {constant=<rgba 0 0 1 1>}
  n1 = ParticleColor()
  n2 = DynamicParameter() {parameterIndex=1, defaultValue=<rgba 0 0 0 1>, paramNames [Index_1,Index_2,Param3,Param4]}
  n3 = Constant() {r=2}
  n4 = TextureObject() {samplerType=SAMPLERTYPE_Masks, texture=&/T_Pickaxe_SilverBullet_Graffitis}
  n5 = TextureSample(
   UVs=MaterialFunctionCall("Animation  Phase (0-1)"=@n2.Index_2, Number of Rows=@n3, Number of Columns=@n3) {materialFunction=&FlipBook}.UVs
   Tex=@n4
  ) {samplerType=SAMPLERTYPE_Masks, texture=&T_Pickaxe_SilverBullet_Graffitis}
  n6 = SmoothStep(Value=@n5.B) {constMax=0.4}
  n7 = ScalarParameter() {defaultValue=1, parameterName=Appearance, sliderMax=1}
  n8 = DynamicParameter() {defaultValue=<rgba 0 0 0 1>, paramNames [Erosion, Appearance, X Vector, Z Vector]}
  n9 = OneMinus(None=Saturate(None=Add(A=@n7, B=@n8.Appearance)))
  n10 = ScalarParameter() {defaultValue=0.2, parameterName=Smoothness, sliderMax=1, sliderMin=0.01}
  n11 = TextureSample(
   UVs=MaterialFunctionCall("Animation  Phase (0-1)"=@n2.Index_1, Number of Rows=@n3, Number of Columns=@n3) {materialFunction=&FlipBook}.UVs
   Tex=@n4
  ) {samplerType=SAMPLERTYPE_Masks, texture=&T_Pickaxe_SilverBullet_Graffitis}
  n12 = SmoothStep(Min=@n9, Max=Saturate(None=Add(A=@n9, B=@n10)), Value=Max(A=@n11.G, B=@n5.A))
  n13 = NamedRerouteUsage()
  n14 = Add(A=ScalarParameter() {parameterName=Erosion, sliderMax=1}, B=@n8.Erosion)
  n15 = ScalarParameter() {defaultValue=0.1, parameterName=Erosion Softness, sliderMax=1, sliderMin=0.01}
  n16 = Multiply(A=OneMinus(None=Saturate(None=@n14)), B=Add(A=@n15))
  n17 = Divide(A=Subtract(A=Saturate(None=@n13), B=Subtract(A=@n16, B=@n15)), B=@n15)
  Multiply(A=@n1.A, B=Subtract(A=Max(A=@n6, B=@n12), B=Saturate(None=@n17)))
  n18 = CurveAtlasRowParameter(CurveTime=@n5.B) {
   atlas=&/CA_Pickaxe_SilverBullet
   curve=&/CC_Pickaxe_SilverBullet_Background
   defaultValue=1
   parameterName=BG_Color
  }
  n19 = CurveAtlasRowParameter(CurveTime=@n11.R) {
   atlas=&CA_Pickaxe_SilverBullet
   curve=&/CC_Pickaxe_SilverBullet_PurpleBlue
   parameterName=TextColor
  }
  n20 = LinearInterpolate(A=@n18, B=@n19, Alpha=@n12)
  Multiply(A=MaterialFunctionCall(Emissive=Multiply(A=@n20, B=@n1.RGB)) {materialFunction=&MF_FX_EmissiveAdjustments}.Result) {constB=0.5}
  n21 = Constant3Vector() {constant=<rgba 1 0 1 0>}
  n22 = DynamicParameter() {defaultValue=<rgba 0 0 0 1>, paramNames [Erosion, Appearance, X Vector, Z Vector]}
  n23 = MaterialFunctionCall(R=@n22.X Vector, G=Constant(), B=@n22.X Vector) {materialFunction=&AppendMany}
  n24 = CrossProduct(A=@n21, B=@n23.RGB)
  n25 = If(A=DotProduct(A=@n24, B=Constant3Vector() {constant=<rgba 0 1 0 0>}), "A > B"=Constant() {r=1}, "A < B"=Constant() {r=-1})
  n26 = Arctangent2Fast(
   Y=MaterialFunctionCall(Vector 3=@n24) {materialFunction=/Engine/Functions/Engine_MaterialFunctions02/Utility/VectorLength}.V3 Length
   X=DotProduct(A=@n21, B=@n23.RGB)
  )
  n27 = TextureSampleParameter2D(UVs=Rotator(Time=Multiply(A=@n25, B=@n26)) {speed=1}) {
   parameterName=Erosion Texture
   samplerType=SAMPLERTYPE_LinearGrayscale
   texture=&/T_Pickaxe_SilverBullet_Graffiti_Erosion
  }
  NamedRerouteDeclaration(None=@n27.R) {name=Erosion Tex, nodeColor=<rgba 1 0 0.6797 1>}
 ]
}
