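# Niagara impact effect for the Silver Bullet pickaxe: a brief burst of colored smoke, layered flares with a light, a graffiti sprite and spike meshes.
# Big_Smoke places partitioned puffs along a system direction vector, while SwingSectionIndex selects their offset and the graffiti material's parameters.
# Front-end modifiers rescale the effect and switch placement and facing to camera vectors.
# Time-of-day values modulate brightness; the splat and FlashLWE emitters have no renderers.
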NiagaraSystem /BRCosmetics/Weapons/FORT_Melee/Pickaxe_SilverBullet/FX/NS_Pickaxe_SilverBullet_Impact
dependencies {
 usedDataInterfaces [
  &NiagaraDataInterfaceCurve
  &NiagaraDataInterfaceColorCurve
  &NiagaraDataInterfaceCamera
  &NiagaraDataInterfaceVector2DCurve
 ]
 usedDynamicInputs [&RemapRange, &FN_FrontEndVector2Modifier]
 usedModules [
  &self:SystemSpawnScript.SystemScriptSource.SystemScriptGraph.NiagaraNodeAssignment_1.SetVariables_#1
  &self:SystemSpawnScript.SystemScriptSource.SystemScriptGraph.NiagaraNodeAssignment_3.SetVariables_#2
  &SystemState
  &self:SystemSpawnScript.SystemScriptSource.SystemScriptGraph.NiagaraNodeAssignment_0.SetVariables_#3
 ]
}
emitters {
 BG_Smoke {
  emitterUpdateScript [
   &EmitterState()
   &SpawnBurst_Instantaneous(Spawn Count=1, Spawn Time=0.03, Use Spawn Probability=false)
   /SetVariables(Emitter.RandomColor:NiagaraInt32=&RandomRangeInt(Minimum=0, Maximum=2)) {enabled=false, isSetParameters=true}
  ]
  particleSpawnScript [
   &FN_SetSpawnGroup(Number of Spawn Groups=2)
   /SetVariables(Particles.Random=&RandomRangeFloat(Minimum=0, Maximum=1)) {isSetParameters=true}
   /Niagara/Modules/Spawn/Initialization/InitializeParticle(
    Color=<rgba 6 6 6 0.3>
    Lifetime=&Lerp_Float(A=0.35, B=0.3, Alpha=$Particles.Random)
    Module.Write Color=true
    Module.Write Lifetime=true
    Module.Write Mass=true
    Module.Write Material Random=true
    Module.Write Position=false
    Module.Write Scale=false
    Module.Write SpriteRotation=true
    Module.Write SpriteSize=true
    Position=&SimulationPosition()
    Sprite Rotation=&Lerp_Float(A=-45, B=45, Alpha=$Particles.Random)
    Sprite Size=&Vector2DFromFloat(Value=&Lerp_Float(A=120, B=130, Alpha=$Particles.Random))
   ) {enabled=false}
   /Niagara/Modules/Spawn/Initialization/V2/InitializeParticle(
    Color=<rgba 6 6 6 0.3>
    Lifetime=&Lerp_Float(A=0.35, B=0.3, Alpha=$Particles.Random)
    Sprite Rotation Mode=::NewEnumerator2
    Sprite Size Mode=::NewEnumerator3
    Uniform Sprite Size=&Multiply_Float(
     A=&RandomRangeFloat(Minimum=120, Maximum=130)
     B=&FN_FrontEndFloatModifier(Gameplay=1, Front End=$User.FrontEndScale)
    )
    UsePositionOffset=false
    Write Material Random=true
   ) {module=InitializeParticle001}
   &CameraOffset(Camera Offset Amount=30)
   &DynamicMaterialParameters(
    Index 0 Param 2=&UniformAOrBFloat(A=0)
    Module.Param0WriteEnabled=true
    Module.Param1WriteEnabled=true
    Module.Param2WriteEnabled=true
    Module.Param3WriteEnabled=true
   ) {module=DynamicMaterialParameters001}
  ]
  particleUpdateScript [
   &ParticleState()
   &ScaleSpriteSize(Uniform Curve Sprite Scale=DataInterface() {propertyValues {Curve {keys [
        {arriveTangent=1.9745, interpMode=RCIM_Cubic, leaveTangent=1.9745, tangentMode=RCTM_User, time=0, value=0.5}
        {arriveTangent=0.5, interpMode=RCIM_Cubic, leaveTangent=0.5, time=0.3, value=0.95}
        <at 1 1 RCIM_Cubic>
       ]}}})
   &SolveForcesAndVelocity(Clamp Velocity=false, Limit Acceleration=false)
   &DynamicMaterialParameters(
    Index 0 Param 1=&FloatFromCurve(FloatCurve=DataInterface() {propertyValues {Curve {keys [<at 0 0>,<at 1 1>]}}})
    Index 0 Param 2=1
    Module.Param0WriteEnabled=true
    Module.Param1WriteEnabled=true
    Module.Param2WriteEnabled=true
    Module.Param3WriteEnabled=true
   )
   &ScaleColor(
    Scale Alpha=&FloatFromCurve(
     FloatCurve=DataInterface() {propertyValues {Curve {keys [<at 0 1>,<at 0.8091 0.9932>,<at 1 0>]}}}
     Scale Curve=&Multiply_Float(A=2, B=$System.TOD)
    )
    Scale RGB=&MakeVectorFromLinearColorRGB(Color=&ColorFromCurve(ColorCurve=DataInterface() {propertyValues {
        AlphaCurve {keys [<at 0 1 RCIM_Cubic>,<at 1 0 RCIM_Cubic>]}
        BlueCurve {keys [<at 0.2863 0.7589 RCIM_Cubic>,<at 0.7871 1 RCIM_Cubic>]}
        GreenCurve {keys [<at 0.2863 1 RCIM_Cubic>,<at 0.7871 0.0292 RCIM_Cubic>]}
        RedCurve {keys [<at 0.2863 0 RCIM_Cubic>,<at 0.7871 0.2886 RCIM_Cubic>]}
       }}))
    ScaleA=true
    ScaleRGB=true
   )
  ]
  rendererClasses [&NiagaraSpriteRendererProperties]
  renderers [{rendererClass=&NiagaraSpriteRendererProperties}]
 }
 Big_Smoke {
  emitterUpdateScript [
   &EmitterState()
   &SpawnRate(SpawnGroup=1, SpawnRate=120, Use Spawn Probability=false)
  ]
  particleSpawnScript [
   &PartitionParticles(Number Of Partitions=5, Particles Per Partition=5, Partition Mode:ENiagara_PartitionMode_v2=::NewEnumerator1)
   &InitializeParticle(
    Color=&MultiplyLinearColor(Linear Color=<rgba 1 1 1 0.5>, Scale Factor=&MakeVector4(X=2, Y=2, Z=2, W=2))
    Lifetime=&RandomRangeFloat(Minimum=0.2, Maximum=0.4)
    Sprite Rotation Angle=&RandomRangeFloat(Minimum=-1, Maximum=1)
    Sprite Rotation Mode=::NewEnumerator1
    Sprite Size Mode=::NewEnumerator3
    Uniform Sprite Size=&Multiply_Float(
     A=&RandomRangeFloat(Minimum=40, Maximum=60)
     B=&FN_FrontEndFloatModifier(Gameplay=1, Front End=$User.FrontEndScale)
    )
    UsePositionOffset=true
    Write Material Random=true
   )
   &ApplyInitialForces()
   /SetVariables(Particles.Position=&AddVectorToPosition(
     Position=$Engine.Emitter.SimulationPosition
     Vector=&Add_Vector(
      A=&FN_FrontEndVector3Modifier(
       Gameplay=&RotateVectorByQuaternion(
        Vector To Rotate=&Add_Vector(
         A=&Multiply_VectorByFloat(
          Vector=$System.DirectionVector
          Float=&Waveform(
           Global Phase=&MakeFloatFromInt(Int=&ReturnParticleIDAttributes())
           "Waveform [1]"=::NewEnumerator0
           "[1] Frequency"=0.1
           "[1] Amplitude Min/Max"=<0 1>
           "[1] Scale"=&Multiply_Float(A=60, B=&FN_FrontEndFloatModifier(Gameplay=1, Front End=$User.FrontEndScale))
          )
         )
         B=&Multiply_VectorByFloat(
          Vector=/Niagara/DynamicInputs/Vectors/CrossProduct(Vector A:Vector3f=$System.DirectionVector, Vector B:Vector3f=<1 0 0>, Normalize Incoming Vectors:NiagaraBool=true)
          Float=&Multiply_FloatByInt(
           Float=&Multiply_Float(A=22, B=&FN_FrontEndFloatModifier(Gameplay=1, Front End=$User.FrontEndScale))
           Integer=$Particles.PartitionParticles.Partition
          )
         )
        )
        Quaternion=$Engine.Owner.Rotation
       )
       Front End=&Add_Vector(
        A=&Multiply_VectorByFloat(
         Vector=&ReturnCameraVector(Camera Vectors=::NewEnumerator2, Coordinate Space=::World)
         Float=&Waveform(
          Global Phase=&MakeFloatFromInt(Int=&ReturnParticleIDAttributes())
          "Waveform [1]"=::NewEnumerator0
          "[1] Frequency"=0.1
          "[1] Amplitude Min/Max"=<0 1>
          "[1] Scale"=&Multiply_Float(A=60, B=&FN_FrontEndFloatModifier(Gameplay=1, Front End=$User.FrontEndScale))
         )
        )
        B=&Multiply_VectorByFloat(
         Vector=&ReturnCameraVector(Camera Vectors=::NewEnumerator3, Coordinate Space=::World)
         Float=&Multiply_FloatByInt(
          Float=&Multiply_Float(A=22, B=&FN_FrontEndFloatModifier(Gameplay=1, Front End=$User.FrontEndScale))
          Integer=$Particles.PartitionParticles.Partition
         )
        )
       )
      )
      B=&Multiply_VectorByFloat(
       Vector=&FN_FrontEndVector3Modifier(
        Gameplay=&RotateVectorByQuaternion(
         Vector To Rotate=&MergeHelper_VectorSwitch(0=<0 -50 0>, 1=<0 0 50>, 2=?, 3=?, Index=$User.SwingSectionIndex, Num=::NewEnumerator0)
         Quaternion=$Engine.Owner.Rotation
        )
        Front End=&Multiply_VectorByFloat(
         Vector=&NormalizeVector(
          Vector To Normalize=&Add_Vector(
           A=&ReturnCameraVector(
            Camera Vectors=::NewEnumerator3
            Coordinate Space=::Simulation
            CameraQuery=DataInterface() {propertyValues {bRequireCurrentFrameData=false}}
           )
           B=&ReturnCameraVector(
            Camera Vectors=::NewEnumerator2
            Coordinate Space=::Simulation
            CameraQuery=DataInterface() {propertyValues {bRequireCurrentFrameData=false}}
           )
          )
          Safe Normalize=true
          Safe Fallback Threshold=0
          Safe Fallback Vector=<0 0 0.001>
         )
         Float=-50
        )
       )
       Float=&FN_FrontEndFloatModifier(Gameplay=1, Front End=$User.FrontEndScale)
      )
     )
    )) {isSetParameters=true}
   &CameraOffset(Camera Offset Amount=50)
  ]
  particleUpdateScript [
   &DynamicMaterialParameters(
    Index 0 Param 1=&Add_Float(
     A=$Particles.RandomFloat
     B=&FloatFromCurve(FloatCurve=DataInterface() {propertyValues {Curve {keys [
          {arriveTangent=0.1552, interpMode=RCIM_Cubic, leaveTangent=0.1552, tangentMode=RCTM_User, time=0, value=1}
          {interpMode=RCIM_Cubic, leaveTangent=0.5473, time=1, value=1.1}
         ]}}})
    )
    Index 0 Param 4=2
    Index 2 Param 1=&FloatFromCurve(FloatCurve=DataInterface() {propertyValues {Curve {keys [
         {interpMode=RCIM_Cubic, leaveTangent=1.6214, tangentMode=RCTM_Break, time=0.1, value=0}
         {arriveTangent=0.8413, tangentMode=RCTM_User, time=1, value=1}
        ]}}})
    Index 2 Param 2=&FloatFromCurve(
     FloatCurve=DataInterface() {propertyValues {Curve {keys [{arriveTangent=0.7845, interpMode=RCIM_Cubic, leaveTangent=0.7845, tangentMode=RCTM_User, time=0, value=0.2}, <at 1 0.8 RCIM_Cubic>]}}}
     Scale Curve=&RandomRangeFloat(Minimum=0.2, Maximum=1.5)
    )
    Index 2 Param 3=&FloatFromCurve(
     FloatCurve=DataInterface() {propertyValues {Curve {keys [
         {interpMode=RCIM_Cubic, leaveTangent=1.0458, tangentMode=RCTM_Break, time=0, value=1}
         {interpMode=RCIM_Cubic, leaveTangent=11.1233, tangentMode=RCTM_Break, time=1, value=2}
        ]}}}
     Scale Curve=1.25
    )
    Index 2 Param 4=&FloatFromCurve(
     FloatCurve=DataInterface() {propertyValues {Curve {keys [<at 0 1 RCIM_Cubic>, {interpMode=RCIM_Cubic, leaveTangent=-1.2956, tangentMode=RCTM_Break, time=1, value=0.75}]}}}
     Scale Curve=1.5
    )
    Index 3 Param 1=&FloatFromCurve(
     FloatCurve=DataInterface() {propertyValues {Curve {keys [
         <at 0.5 0 RCIM_Cubic>
         {arriveTangent=1, interpMode=RCIM_Cubic, leaveTangent=1, time=1, value=1}
         {arriveTangent=-1, interpMode=RCIM_Cubic, leaveTangent=-1, time=1.5, value=1}
         {interpMode=RCIM_Cubic, leaveTangent=2, time=2, value=0}
        ]}}}
     CurveIndex=$Emitter.Age
     Scale Curve=&Multiply_Float(
      A=&FN_Select_Float_from_Time_Of_Day(Morning=3, Day=3, Evening=3, Night=3)
      B=&FN_FrontEndFloatModifier(Gameplay=0, Front End=2)
     )
    )
    Index 3 Param 2=&Multiply_Float(
     A=&Lerp_Float(
      A=33
      B=75
      Alpha=&FloatFromCurve(
       FloatCurve=DataInterface() {propertyValues {Curve {keys [
           <at 0.5 0 RCIM_Cubic>
           {arriveTangent=1, interpMode=RCIM_Cubic, leaveTangent=1, time=1, value=1}
           {arriveTangent=-1, interpMode=RCIM_Cubic, leaveTangent=-1, time=1.5, value=1}
           <at 2 0 RCIM_Cubic>
          ]}}}
       CurveIndex=$Emitter.Age
      )
     )
     B=&FN_FrontEndFloatModifier(Gameplay=1, Front End=1.25)
    )
    Module.Param0WriteEnabled=true
    Module.Param0WriteEnabled2=true
    Module.Param0WriteEnabled3=true
    Module.Param1WriteEnabled=false
    Module.Param1WriteEnabled2=true
    Module.Param1WriteEnabled3=true
    Module.Param2WriteEnabled=true
    Module.Param2WriteEnabled2=true
    Module.Param2WriteEnabled3=true
    Module.Param3WriteEnabled=true
    Module.Param3WriteEnabled2=true
    Module.Param3WriteEnabled3=true
    Write Parameter Index 2=true
    Write Parameter Index 3=true
   )
   &ParticleState()
   &ScaleSpriteSize(Uniform Curve Sprite Scale=DataInterface() {propertyValues {Curve {keys [{arriveTangent=4.2408, interpMode=RCIM_Cubic, leaveTangent=4.2408, tangentMode=RCTM_User, time=0, value=0.5}, <at 1 2 RCIM_Cubic>]}}}) {module=ScaleSpriteSize001}
   &ScaleColor(
    Scale Alpha=&FloatFromCurve(
     FloatCurve=DataInterface() {propertyValues {Curve {keys [
         <at 0 1.2 RCIM_Cubic>
         {arriveTangent=-2.6917, interpMode=RCIM_Cubic, leaveTangent=-2.6917, tangentMode=RCTM_User, time=0.3, value=1}
         {tangentMode=RCTM_User, time=1, value=0}
        ]}}}
     Scale Curve=0.6
    )
    Scale RGB=&MakeVectorFromLinearColorRGB(Color=&MergeHelper_ColorSwitch(
      0=&ColorFromCurve(ColorCurve=DataInterface() {propertyValues {
         AlphaCurve {keys [<at 0.0453 1>]}
         BlueCurve {keys [<at 0 1>,<at 1 1>]}
         GreenCurve {keys [<at 0 0.7406>,<at 1 0.7367>]}
         RedCurve {keys [<at 0 0>,<at 1 0>]}
        }})
      1=&ColorFromCurve(ColorCurve=DataInterface() {propertyValues {
         AlphaCurve {keys [<at 0.0453 1>]}
         BlueCurve {keys [<at 0 0.75>,<at 0.9937 0.7552>]}
         GreenCurve {keys [<at 0 0.0494>,<at 0.9937 0.0511>]}
         RedCurve {keys [<at 0 0.1505>,<at 0.9937 0.1527>]}
        }})
      2=?
      3=?
      Index=&RandomRangeInt(Minimum=0, Maximum=1)
      Num=::NewEnumerator0
     ))
    ScaleA=true
    ScaleRGB=true
   )
   &Drag(Drag=&RandomRangeFloat(Minimum=3, Maximum=4), Use Linear Drag=true, Use Rotational Drag=true)
   &SolveForcesAndVelocity(Clamp Velocity=false, Limit Acceleration=false)
   /SetVariables(Particles.UnlitSmokeColor:LinearColor=&MultiplyLinearColor(
     Linear Color=$Particles.Color
     Scale Factor=&MakeVector4_FromLinearColor(LinearColor=&MultiplyLinearColor(
       Linear Color=<rgba 0.3438 0.3438 0.3438 1>
       Scale Factor=&MakeVector4_FromLinearColor(LinearColor=$System.TOD_Color)
      ))
    )) {isSetParameters=true}
   &KillParticles(Kill Particles=&SetBoolByFloatComparison(A=$Particles.NormalizedAge, B=0.75))
  ]
  rendererClasses [&NiagaraSpriteRendererProperties]
  renderers [
   {rendererClass=&NiagaraSpriteRendererProperties}
   {rendererClass=&NiagaraSpriteRendererProperties, rendererIndex=1}
  ]
 }
 Flare {
  emitterUpdateScript [
   &EmitterState()
   &SpawnBurst_Instantaneous(Spawn Count=1, Use Spawn Probability=false)
   &SpawnBurst_Instantaneous(Spawn Count=1, Spawn Group=1, Use Spawn Probability=false) {module=SpawnBurst_Instantaneous001}
   &SpawnBurst_Instantaneous(Spawn Count=1, Spawn Group=2, Use Spawn Probability=false) {module=SpawnBurst_Instantaneous002}
  ]
  particleSpawnScript [
   /SetVariables(Particles.VisibilityTag=$Emitter.SpawnGroup) {isSetParameters=true}
   &InitializeParticle(
    Color=&MergeHelper_ColorSwitch(
     0=<rgba 5 5 5 0.5>
     1=&MergeHelper_ColorSwitch(
      0=<rgba 0.1946 0.0931 0.521 1>
      1=<rgba 0.071 0.25 0.2342 1>
      2=<rgba 0.2 0.0189 0.1247 1>
      3=?
      Index=&RandomRangeInt(Minimum=0, Maximum=2)
      Num=::NewEnumerator1
     )
     2=<rgba 0.0182 0.008 0.05 2>
     3=?
     Index=$Particles.VisibilityTag
     Num=::NewEnumerator1
    )
    Lifetime=&MergeHelper_FloatSwitch(0=0.25, 1=0.4, 2=0.5, 3=?, Index=$Particles.VisibilityTag, Num=::NewEnumerator1)
    Sprite Rotation Mode=::NewEnumerator2
    Sprite Size Mode=::NewEnumerator3
    Uniform Sprite Size=&Multiply_Float(
     A=&MergeHelper_FloatSwitch(0=120, 1=200, 2=155, 3=?, Index=$Particles.VisibilityTag, Num=::NewEnumerator1)
     B=&FN_FrontEndFloatModifier(Gameplay=1, Front End=$User.FrontEndScale)
    )
    UsePositionOffset=false
    Write Material Random=true
   )
   &DynamicMaterialParameters(
    Index 0 Param 1=&RandomRangeFloat(Minimum=0.4, Maximum=0.6)
    Index 0 Param 2=1
    Index 0 Param 3=1
    Index 0 Param 4=1
    Module.Index0Float0Write=true
    Module.Index0Float1Write=true
    Module.Index0Float2Write=true
    Module.Index0Float3Write=true
   )
   &CameraOffset(Camera Offset Amount=&MakeCustomFloatFromBool(Bool=$User.bIsFrontEnd, True Float=10, False Float=30))
  ]
  particleUpdateScript [
   &ParticleState()
   &ScaleColor(
    Scale Alpha=&MergeHelper_FloatSwitch(
     0=&FloatFromCurve(FloatCurve=DataInterface() {propertyValues {Curve {keys [<at 0 0>,<at 0.05 1>,<at 0.6 0.6>,<at 0.7 -0.05>]}}})
     1=&FloatFromCurve(FloatCurve=DataInterface() {propertyValues {Curve {keys [<at 0 0>,<at 0.05 1>,<at 0.6 0.6>,<at 0.7 -0.05>]}}})
     2=&Multiply_Float(
      A=&FloatFromCurve(
       FloatCurve=DataInterface() {propertyValues {Curve {keys [<at 0 1>,<at 1 0>]}}}
       Scale Curve=&FN_FrontEndFloatModifier(Gameplay=1, Front End=0.5)
      )
      B=&FNTimeOfDayScalar(Night Scale=0.3, Morning Scale=0.05, Day Scale=0.5, Evening Scale=0.05)
     )
     3=?
     Index=$Particles.VisibilityTag
     Num=::NewEnumerator1
    )
    Scale RGB=&MakeVectorFromLinearColorRGB(Color=&MergeHelper_ColorSwitch(
      0=&ColorFromCurve(ColorCurve=DataInterface() {propertyValues {
         AlphaCurve {keys []}
         BlueCurve {keys [<at 0 0.3561>,<at 0.6826 0.3561>,<at 1 1>]}
         GreenCurve {keys [<at 0 1>,<at 0.6826 1>,<at 1 0.2882>]}
         RedCurve {keys [<at 0 0>,<at 0.6826 0>,<at 1 0.3589>]}
        }})
      1=<rgba 1 1 1 1>
      2=<rgba 1 1 1 1>
      3=?
      Index=$Particles.VisibilityTag
      Num=::NewEnumerator1
     ))
    ScaleA=true
    ScaleRGB=true
   )
   &ScaleSpriteSize(Scale Factor=&Vector2DFromFloat(Value=&MergeHelper_FloatSwitch(
      0=&FloatFromCurve(FloatCurve=DataInterface() {propertyValues {Curve {keys [<at 0 1.5 RCIM_Cubic>,<at 0.8 0>]}}}, Scale Curve=1.3)
      1=&FloatFromCurve(
       FloatCurve=DataInterface() {propertyValues {Curve {keys [<at 0 1.5 RCIM_Cubic>,<at 1 0.5>]}}}
       Scale Curve=&FN_FrontEndFloatModifier(Gameplay=1, Front End=$User.FrontEndScale)
      )
      2=&FloatFromCurve(FloatCurve=DataInterface() {propertyValues {Curve {keys [<at 0 1>,<at 1 3>]}}})
      3=?
      Index=$Particles.VisibilityTag
      Num=::NewEnumerator1
     )))
   &FN_Light_Attributes(
    Color=&ColorFromCurve(ColorCurve=DataInterface() {propertyValues {
       AlphaCurve {keys [<at 0 1 RCIM_Cubic>]}
       BlueCurve {keys [<at 0.0184 0.4009>,<at 1 0.25 RCIM_Cubic>]}
       GreenCurve {keys [<at 0.0184 1>,<at 1 0.0345 RCIM_Cubic>]}
       RedCurve {keys [<at 0.0184 0>,<at 1 0.0007 RCIM_Cubic>]}
      }})
    Create Color Attribute=true
    Create Exponent Attribute=true
    Create Position Attribute=true
    Create Radius Attribute=true
    Enable Particle Light Enable=true
    Exponent=3
    Light Enabled=true
    Position=<15 0 0>
    Radius=90
    Scale Color=&Multiply_Float(
     A=&FloatFromCurve(FloatCurve=DataInterface() {propertyValues {Curve {keys [
          {arriveTangent=-4.7352, interpMode=RCIM_Cubic, time=0, value=3}
          {arriveTangent=-3, interpMode=RCIM_Cubic, leaveTangent=-3, time=0.3, value=1}
          {interpMode=RCIM_Cubic, leaveTangent=-0.5426, time=1, value=0}
         ]}}})
     B=&FloatFromCurve(
      FloatCurve=DataInterface() {propertyValues {Curve {keys [<at 0 0.7>,<at 0.25 1>,<at 0.6 1>,<at 1 0.7>]}}}
      CurveIndex=Linked() {linkedVariable {name=NPC.FortniteNPC.TimeOfDay_0-1, type=NiagaraFloat}}
     )
    )
   )
  ]
  rendererClasses [&NiagaraSpriteRendererProperties, &NiagaraLightRendererProperties]
  renderers [
   {rendererClass=&NiagaraSpriteRendererProperties}
   {rendererClass=&NiagaraLightRendererProperties, rendererIndex=1}
   {rendererClass=&NiagaraSpriteRendererProperties, rendererIndex=2}
   {rendererClass=&NiagaraSpriteRendererProperties, rendererIndex=3}
  ]
 }
 FlashLWE {rendererClasses [], renderers []}
 Graffiti {
  emitterUpdateScript [
   &EmitterState()
   &SpawnBurst_Instantaneous(Spawn Count=1, Use Spawn Probability=false)
  ]
  particleSpawnScript [
   &InitializeParticle(
    Color=<rgba 1 1 1 1.2>
    Lifetime=0.6
    Sprite Rotation Angle=&RandomRangeFloat(Minimum=-0.05, Maximum=0.05)
    Sprite Rotation Mode=::NewEnumerator1
    Sprite Size Mode=::NewEnumerator3
    Uniform Sprite Size=&Multiply_Float(
     A=&RandomRangeFloat(Minimum=105, Maximum=115)
     B=&FN_FrontEndFloatModifier(Gameplay=1, Front End=$User.FrontEndScale)
    )
    UsePositionOffset=false
    Write Material Random=true
   )
   &CameraOffset(Camera Offset Amount=30)
   &SpriteFacingAndAlignment(
    Alignment Coordinate Space=::World
    Facing Coordinate Space=::World
    Sprite Alignment=<0 0 1>
    Sprite Facing=&FN_FrontEndVector3Modifier(
     Gameplay=&TransformVector(Vector=<1 0 0>, Destination Space=::World)
     Front End=&Multiply_VectorByFloat(
      Vector=&ReturnCameraVector(Camera Vectors=::NewEnumerator1, Coordinate Space=::World, CameraQuery=DataInterface() {propertyValues {bRequireCurrentFrameData=false}})
      Float=-1
     )
    )
    WriteAlignment=true
    WriteFacing=true
   )
   &DynamicMaterialParameters(
    Index 1 Param 1=&RandomRangeFloat(Minimum=0.25, Maximum=0.99)
    Index 1 Param 2=&RandomRangeFloat(Minimum=0, Maximum=1)
    Module.Index1Float0Write=true
    Module.Index1Float1Write=true
    Module.Index1Float2Write=true
    Module.Index1Float3Write=true
    Write Parameter Index 0=::NewEnumerator0
    Write Parameter Index 1=::NewEnumerator1
   ) {module=DynamicMaterialParameters001}
  ]
  particleUpdateScript [
   &DynamicMaterialParameters(
    Index 0 Param 1=&FloatFromCurve(FloatCurve=DataInterface() {propertyValues {Curve {keys [
         <at 0 1 RCIM_Cubic>
         {arriveTangent=-1.8, interpMode=RCIM_Cubic, leaveTangent=-1.8, time=0.1, value=0.1}
         {arriveTangent=0.4, interpMode=RCIM_Cubic, leaveTangent=0.4, time=0.5, value=0.1}
         {arriveTangent=1.1648, interpMode=RCIM_Cubic, leaveTangent=1.1648, tangentMode=RCTM_User, time=0.6, value=0.3}
         {arriveTangent=2.65, interpMode=RCIM_Cubic, leaveTangent=2.65, tangentMode=RCTM_User, time=1, value=0.9}
        ]}}})
    Index 0 Param 2=&FloatFromCurve(FloatCurve=DataInterface() {propertyValues {Curve {keys [<at 0.3 0.2>,<at 0.9 1>]}}})
    Index 0 Param 3=&MergeHelper_FloatSwitch(0=1, 1=-1, 2=?, 3=?, Index=$User.SwingSectionIndex, Num=::NewEnumerator0)
    Index 0 Param 4=&MergeHelper_FloatSwitch(0=-1, 1=-1, 2=?, 3=?, Index=$User.SwingSectionIndex, Num=::NewEnumerator0)
    Module.Index0Float0Write=true
    Module.Index0Float1Write=true
    Module.Index0Float2Write=true
    Module.Index0Float3Write=true
   )
   &ParticleState()
   &ScaleColor(
    Scale Alpha=&FloatFromCurve(FloatCurve=DataInterface() {propertyValues {Curve {keys [<at 0 0>,<at 0.1 1>,<at 1 1>]}}})
    ScaleA=true
    ScaleRGB=true
   )
   &ScaleSpriteSize(Scale Factor=&Vector2DFromFloat(Value=&FloatFromCurve(FloatCurve=DataInterface() {propertyValues {Curve {keys [<at 0.02 0>,<at 0.1 1 RCIM_Cubic>]}}})))
   /SetVariables(Particles.UnlitSmokeColor:LinearColor=&MultiplyLinearColor(
     Linear Color=$Particles.Color
     Scale Factor=&MakeVector4_FromLinearColor(LinearColor=&MultiplyLinearColor(Linear Color=<rgba 3 3 3 1>, Scale Factor=&MakeVector4_FromLinearColor(LinearColor=$System.TOD_Color)))
    )) {isSetParameters=true}
  ]
  rendererClasses [&NiagaraSpriteRendererProperties]
  renderers [
   {rendererClass=&NiagaraSpriteRendererProperties}
   {rendererClass=&NiagaraSpriteRendererProperties, rendererIndex=1}
  ]
 }
 Impact_Spikes {
  emitterUpdateScript [
   &EmitterState()
   &SpawnBurst_Instantaneous(Spawn Count=1, Use Spawn Probability=false)
  ]
  particleSpawnScript [
   &InitializeParticle(
    Color=<rgba 0 3 2.1601 5>
    Lifetime=&RandomRangeFloat(Minimum=0.2, Maximum=0.25)
    Mesh Scale=&VectorFromFloat(Value=&Multiply_Float(
      A=&RandomRangeFloat(Minimum=1.3, Maximum=1.7)
      B=&FN_FrontEndFloatModifier(Gameplay=1, Front End=$User.FrontEndScale)
     ))
    Mesh Scale Mode=::NewEnumerator4
    UsePositionOffset=false
    Write Material Random=true
   ) {module=InitializeParticle001}
   &CameraOffset(Camera Offset Amount=50)
   &InitialMeshOrientation(
    Rotation=&RandomRangeVector(Minimum=<-1 -1 -1>, Maximum=<1 1 1>)
    Use Orientation Vector=true
    Use Rotation Vector=true
   ) {module=InitialMeshOrientation001}
  ]
  particleUpdateScript [
   &ParticleState()
   &ScaleColor(
    Scale Alpha=&FloatFromCurve(FloatCurve=DataInterface() {propertyValues {Curve {keys [{arriveTangent=-2.6501, interpMode=RCIM_Cubic, leaveTangent=-2.6501, tangentMode=RCTM_User, time=0, value=1}, <at 1 0>]}}})
    ScaleA=true
    ScaleRGB=false
   ) {enabled=false}
   &ScaleMeshSize(Scale Factor=&VectorFromFloat(Value=&FloatFromCurve(
      FloatCurve=DataInterface() {propertyValues {Curve {keys [{arriveTangent=1.5592, interpMode=RCIM_Cubic, leaveTangent=1.5592, tangentMode=RCTM_User, time=0, value=0.35}, <at 0.6 1.25>]}}}
      Scale Curve=&FN_FrontEndFloatModifier(Gameplay=1, Front End=0.6)
     )))
   &DynamicMaterialParameters(
    Index 0 Param 1=&FloatFromCurve(FloatCurve=DataInterface() {propertyValues {Curve {keys [<at 0 1 RCIM_Cubic>,<at 0.8 0 RCIM_Cubic>]}}})
    Index 0 Param 2=2
    Index 0 Param 3=1
    Index 0 Param 4=1
    Index 1 Param 1=&FloatFromCurve(FloatCurve=DataInterface() {propertyValues {Curve {keys [
         {arriveTangent=-2.3163, interpMode=RCIM_Cubic, time=0, value=2}
         {arriveTangent=-1.9019, interpMode=RCIM_Cubic, leaveTangent=-1.9019, tangentMode=RCTM_User, time=1, value=0.4}
        ]}}})
    Index 1 Param 2=0.4
    Index 1 Param 3=0.6
    Index 1 Param 4=1
    Index 2 Param 1=&FloatFromCurve(FloatCurve=DataInterface() {propertyValues {Curve {keys [<at 0.46 0>,<at 1 1>]}}})
    Index 2 Param 2=45
    Module.Param0WriteEnabled=true
    Module.Param0WriteEnabled1=true
    Module.Param0WriteEnabled2=true
    Module.Param1WriteEnabled=true
    Module.Param1WriteEnabled1=true
    Module.Param1WriteEnabled2=true
    Module.Param2WriteEnabled=true
    Module.Param2WriteEnabled1=true
    Module.Param2WriteEnabled2=true
    Module.Param3WriteEnabled=true
    Module.Param3WriteEnabled1=true
    Module.Param3WriteEnabled2=true
    Write Parameter Index 1=true
    Write Parameter Index 2=true
   ) {module=DynamicMaterialParameters001}
   &FN_Mesh_Attributes(
    Color=<rgba 0.0581 0.0412 0.4167 5>
    Create Color Attribute=true
    Create Scale Attribute=true
    Scale=&Multiply_VectorByFloat(Vector=$Particles.Scale, Float=1.2)
    Write to Module Namespace=true
   )
   &ScaleColor(Scale RGB=&VectorFromFloat(Value=$System.TOD), ScaleA=true, ScaleRGB=true) {module=ScaleColor001}
  ]
  rendererClasses [&NiagaraMeshRendererProperties]
  renderers [
   {rendererClass=&NiagaraMeshRendererProperties, rendererIndex=0}
   {rendererClass=&NiagaraMeshRendererProperties, rendererIndex=1}
  ]
 }
 Splat {rendererClasses [], renderers []}
 SplatSmall {rendererClasses [], renderers []}
 SplatSmall001 {rendererClasses [], renderers []}
}
summary {
 systemName=NS_Pickaxe_SilverBullet_Impact
 emitters [
  {emitterName=BG_Smoke, rendererClasses [&NiagaraSpriteRendererProperties]}
  {emitterName=Impact_Spikes, rendererClasses [&NiagaraMeshRendererProperties]}
  {emitterName=Graffiti, rendererClasses [&NiagaraSpriteRendererProperties]}
  {emitterName=Flare, rendererClasses [&NiagaraSpriteRendererProperties, &NiagaraLightRendererProperties]}
  {emitterName=Splat, rendererClasses []}
  {emitterName=SplatSmall, rendererClasses []}
  {emitterName=FlashLWE, rendererClasses []}
  {emitterName=SplatSmall001, rendererClasses []}
  {emitterName=Big_Smoke, rendererClasses [&NiagaraSpriteRendererProperties]}
 ]
 userVariables [
  User.bIsFrontEnd {type=NiagaraBool}
  User.bIsFrontEndPreview {type=NiagaraBool}
  User.FrontEndScale {type=NiagaraFloat, defaultValue=0.4, description=""}
  User.SwingSectionIndex {type=NiagaraInt32}
 ]
}
